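# Physics asset for the dynamic hair of the F_MED_Aztec_Moon female player body.
# Five three-bone chains of box bodies (left, right, back, back-left, back-right) hang from the head with limited swing and tight twist, looser down the center-back chain.
# Sphere and capsule bodies cover head, neck, upper spine, clavicles and, asymmetrically, only the right upper-arm twist bone.
# Head, neck, spine and clavicles link through free-rotation constraints; the arm body has none.
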PhysicsAsset /BRCosmetics/Characters/Player/Female/Medium/Bodies/F_MED_Aztec_Moon/Meshes/Parts/F_MED_Aztec_Moon_dyn_hair
bodies {
 clavicle_l [{shapeType=Capsule, length=4.875, radius=6.66, center=<6.3404 -0.9148 -0.8034>, rotation=<rot 69.1488 0 0>}]
 clavicle_r [{shapeType=Capsule, length=4.875, radius=6.66, center=<-6.3403 0.9148 0.8034>, rotation=<rot 70.1598 0 0>}]
 dyn_hair_bk_01 [{shapeType=Box, extentX=5.9004, extentY=1.8791, extentZ=7.5124, center=<3.2348 0 -0.0001>}]
 dyn_hair_bk_02 [{shapeType=Box, extentX=10.0425, extentY=1.8791, extentZ=7.5124, center=<5.0038 0 -0.0001>}]
 dyn_hair_bk_03 [{shapeType=Box, extentX=12.4028, extentY=2.5131, extentZ=7.5124, center=<6.3762 0.5144 0.0001>, rotation=<rot 0 5.0002 0>}]
 dyn_hair_bk_l_01 [{shapeType=Box, extentX=5.9004, extentY=1.8791, extentZ=7.5124, center=<3.2348 0 -0.0001>}]
 dyn_hair_bk_l_02 [{shapeType=Box, extentX=5.9004, extentY=1.8791, extentZ=7.5124, center=<3.2348 0 -0.0001>}]
 dyn_hair_bk_l_03 [{shapeType=Box, extentX=1.5258, extentY=1.8791, extentZ=5.4635, center=<0.8777 0 -0.0001>}]
 dyn_hair_bk_r_01 [{shapeType=Box, extentX=5.9004, extentY=1.8791, extentZ=7.5124, center=<3.2348 0 -0.0001>}]
 dyn_hair_bk_r_02 [{shapeType=Box, extentX=5.9004, extentY=1.8791, extentZ=7.5124, center=<3.2348 0 -0.0001>}]
 dyn_hair_bk_r_03 [{shapeType=Box, extentX=2.6368, extentY=3.8, extentZ=7.5124, center=<1.531 -0.0001 -0.0001>}]
 dyn_hair_l_01 [{shapeType=Box, extentX=5.9004, extentY=1.6203, extentZ=7.5124, center=<3.2348 0 -0.0001>}]
 dyn_hair_l_02 [{shapeType=Box, extentX=5.2096, extentY=1.8791, extentZ=7.5124, center=<3.2348 0 -0.0001>}]
 dyn_hair_l_03 [{shapeType=Box, extentX=1.7866, extentY=1.8791, extentZ=7.5124, center=<1.1786 0 -0.0001>}]
 dyn_hair_r_01 [{shapeType=Box, extentX=5.9004, extentY=1.8791, extentZ=7.5124, center=<3.2348 0 -0.0001>}]
 dyn_hair_r_02 [{shapeType=Box, extentX=5.9004, extentY=1.8791, extentZ=7.5124, center=<3.2348 0 -0.0001>}]
 dyn_hair_r_03 [{shapeType=Box, extentX=5.3718, extentY=2.5, extentZ=7.5124, center=<1.997 -1.6449 0>, rotation=<rot 0 -52.8346 0>}]
 head [
  {shapeType=Sphere, radius=6.6677, center=<1.3281 -0.8773 0>}
  {shapeType=Capsule, length=5.9657, radius=5.522, center=<0.8117 2.2232 0>, rotation=<rot 90 0 0>}
 ]
 neck_01 [{shapeType=Capsule, length=8.2568, radius=5.3864, rotation=<rot 90 0 0>}]
 spine_05 [{shapeType=Capsule, length=7.3601, radius=12.0816, center=<0.9683 1.6 -0.0289>, rotation=<rot 5 0 0>}]
 upperarm_twist_01_r [{shapeType=Capsule, length=20.8506, radius=6.66, center=<-4.7271 0.0001 1.0829>, rotation=<rot 86.8403 180 180>}]
}
constraints [
 {bone1Name=dyn_hair_bk_l_01, bone2Name=head, swing1LimitDegrees=45, swing2LimitDegrees=45, twistLimitDegrees=5}
 {bone1Name=dyn_hair_bk_l_02, bone2Name=dyn_hair_bk_l_01, swing1LimitDegrees=45, swing2LimitDegrees=45, twistLimitDegrees=5}
 {bone1Name=dyn_hair_bk_l_03, bone2Name=dyn_hair_bk_l_02, swing1LimitDegrees=45, swing2LimitDegrees=45, twistLimitDegrees=5}
 {bone1Name=dyn_hair_bk_01, bone2Name=head, swing1LimitDegrees=45, swing2LimitDegrees=45, twistLimitDegrees=5}
 {bone1Name=dyn_hair_bk_02, bone2Name=dyn_hair_bk_01, swing1LimitDegrees=60, swing2LimitDegrees=60, twistLimitDegrees=5}
 {bone1Name=dyn_hair_bk_03, bone2Name=dyn_hair_bk_02, swing1LimitDegrees=60, swing2LimitDegrees=60, twistLimitDegrees=5}
 {bone1Name=dyn_hair_l_01, bone2Name=head, swing1LimitDegrees=45, swing2LimitDegrees=45, twistLimitDegrees=5}
 {bone1Name=dyn_hair_l_02, bone2Name=dyn_hair_l_01, swing1LimitDegrees=45, swing2LimitDegrees=45, twistLimitDegrees=5}
 {bone1Name=dyn_hair_l_03, bone2Name=dyn_hair_l_02, swing1LimitDegrees=45, swing2LimitDegrees=45, twistLimitDegrees=5}
 {bone1Name=dyn_hair_bk_r_01, bone2Name=head, swing1LimitDegrees=45, swing2LimitDegrees=45, twistLimitDegrees=5}
 {bone1Name=dyn_hair_bk_r_02, bone2Name=dyn_hair_bk_r_01, swing1LimitDegrees=45, swing2LimitDegrees=45, twistLimitDegrees=5}
 {bone1Name=dyn_hair_bk_r_03, bone2Name=dyn_hair_bk_r_02, swing1LimitDegrees=45, swing2LimitDegrees=45, twistLimitDegrees=5}
 {bone1Name=dyn_hair_r_01, bone2Name=head, swing1LimitDegrees=45, swing2LimitDegrees=45, twistLimitDegrees=5}
 {bone1Name=dyn_hair_r_02, bone2Name=dyn_hair_r_01, swing1LimitDegrees=45, swing2LimitDegrees=45, twistLimitDegrees=5}
 {bone1Name=dyn_hair_r_03, bone2Name=dyn_hair_r_02, swing1LimitDegrees=45, swing2LimitDegrees=45, twistLimitDegrees=5}
 {
  bone1Name=clavicle_l
  bone2Name=spine_05
  swing1LimitDegrees=45
  swing1Motion=Free
  swing2LimitDegrees=45
  swing2Motion=Free
  twistLimitDegrees=45
  twistMotion=Free
 }
 {
  bone1Name=clavicle_r
  bone2Name=spine_05
  swing1LimitDegrees=45
  swing1Motion=Free
  swing2LimitDegrees=45
  swing2Motion=Free
  twistLimitDegrees=45
  twistMotion=Free
 }
 {
  bone1Name=head
  bone2Name=neck_01
  swing1LimitDegrees=45
  swing1Motion=Free
  swing2LimitDegrees=45
  swing2Motion=Free
  twistLimitDegrees=45
  twistMotion=Free
 }
 {
  bone1Name=neck_01
  bone2Name=spine_05
  swing1LimitDegrees=45
  swing1Motion=Free
  swing2LimitDegrees=45
  swing2Motion=Free
  twistLimitDegrees=45
  twistMotion=Free
 }
]
