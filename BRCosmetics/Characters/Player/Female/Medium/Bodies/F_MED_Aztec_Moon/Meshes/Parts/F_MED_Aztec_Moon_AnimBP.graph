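# Animation Blueprint for the F_MED_Aztec_Moon body part, copying the attached parent's pose and simulating hair physics on top.
# A RigidBody node runs in head space, with lighter gravity while skydive-floating and an external force while skydiving.
# Each update derives that force from the head-to-pelvis direction, scaled by floating, diving or default state.
# Delayed dynamics resets follow parachute opening and skydive or crouch transitions to re-settle the simulation.
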AnimBlueprint /BRCosmetics/Characters/Player/Female/Medium/Bodies/F_MED_Aztec_Moon/Meshes/Parts/F_MED_Aztec_Moon_AnimBP
target_skeleton=&/F_MED_Aztec_Moon_Skeleton
graphs [
 AnimGraph {nodes [
   n1 = RigidBody!(
    ComponentPose=Local To Component(LocalPose=Copy Pose From Mesh() {bUseAttachedParent=true})
    Alpha
    OverrideWorldGravity:Vector=SelectVector(A="0, 0,-100", B="0, 0,-980", bPickA=Get bIsSkydiveFloating())
    ExternalForce=SelectVector(A=Get skydiving vector force() -> (skydiving vector force:Vector), B="0, 0, 0", bPickA=Get bIsSkyDiving())
   ) -> (Pose) {
    bOverrideWorldGravity=true
    overridePhysicsAsset=&/F_MED_Aztec_Moon_dyn_hair
    simulationSpace=BaseBoneSpace
    alphaBoolBlend {blendOption=Linear}
    baseBoneRef {boneName=head}
    componentLinearAccScale=<0.2 0.5 0.8>
    componentLinearVelScale=<0.2 0.5 0.8>
    overrideWorldGravity {z=-980}
   }
   Output Pose(Result=Component To Local(ComponentPose=@n1)) {layerGroup=None}
  ]}
 EventGraph {nodes [
   Event BlueprintUpdateAnimation()
   > n1 = Branch(Condition=Get bIsSkydiveFloating())
   n2 = Get Direction Between Sockets!(
    Component:Skeletal Mesh Component Object Reference=GetOwnerSkeletalMeshComponent()
    SocketOrBoneNameFrom:Name=head
    SocketOrBoneNameTo:Name=pelvis
   ) -> (ReturnValue:Vector) {is=title K2Node_CallFunction}
   > Set skydiving vector force(skydiving vector force:Vector=vector + vector(
     A:Vector=vector * vector(A:Vector=@n2, B:Vector="75,75,75") -> (ReturnValue:Vector) {is=title K2Node_CommutativeAssociativeBinaryOperator}
     B:Vector="0, 0, 0"
    ) -> (ReturnValue:Vector) {is=title K2Node_CommutativeAssociativeBinaryOperator}) -> (Output_Get:Vector)
   n3 = Branch(execute=@n1.else, Condition=Get bIsSkydiveDiveMode())
   > Set skydiving vector force(skydiving vector force=vector + vector(A=vector * vector(A=@n2, B="180,180,200"), B="0, 0, 0"))
   Set skydiving vector force(execute=@n3.else, skydiving vector force=vector * vector(A=@n2, B="-15,-15,-15"))
   Event OnParachuteOpened()
   > n4 = Delay(Duration=1.5)
   ResetDynamics(execute=[@n4.then, @n6.then, @n11.then], InTeleportType=TeleportPhysics)
   > ResetDynamics(InTeleportType=ResetPhysics)
   n5 = Event OnBeginSkydiving()
   n6 = Delay(execute=[@n5.then, @n7.then, @n8.then, @n9.then], Duration=0.125)
   n7 = Event OnEndSkydiving()
   n8 = Event OnBeginCrouching()
   n9 = Event OnEndCrouching()
   n10 = Get bIsCrouching()
   n11 = Branch(Condition=AND Boolean(A=Get bIsTargeting(), B=NOT Boolean(A=@n10)))
   Event OnBeginTargeting() {is=title K2Node_Event}
  ]}
]
variables [skydiving vector force {type=Vector}]
